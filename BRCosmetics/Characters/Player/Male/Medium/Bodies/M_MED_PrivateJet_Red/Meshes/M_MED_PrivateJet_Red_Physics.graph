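# Physics asset for the red PrivateJet male medium body cosmetic, giving its coat tails a jointed physics setup alongside body collision.
# Capsules cover the pelvis, spine, thighs and calves, mostly linked by free-swinging joints, though spine_03 has no constraint.
# Left and right six-segment tail chains hang from small master bodies on spine_02.
# Tail links use tight, uneven swing limits and zero twist, while the master joints rotate freely.
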PhysicsAsset /BRCosmetics/Characters/Player/Male/Medium/Bodies/M_MED_PrivateJet_Red/Meshes/M_MED_PrivateJet_Red_Physics
bodies {
 calf_l [{shapeType=Capsule, length=34.5984, radius=5.9519, shapeName=calf_l_capsule, center=<-17.0449 0 0>, rotation=<rot 90 0 0>}]
 calf_r [{shapeType=Capsule, length=34.5984, radius=5.9519, shapeName=calf_l_capsule, center=<17.0449 0 0>, rotation=<rot 90 180 0>}]
 dyn_coatTail_l_01 [{shapeType=Capsule, length=7.0213, radius=8.764, shapeName=dyn_coatTail_l_01_capsule, center=<7.2381 0 7.7>, rotation=<rot 90 0 0>}]
 dyn_coatTail_l_02 [{shapeType=Capsule, length=7.0213, radius=8.764, shapeName=dyn_coatTail_l_02_capsule, center=<7.2381 0 7.7>, rotation=<rot 90 0 0>}]
 dyn_coatTail_l_03 [{shapeType=Capsule, length=7.0213, radius=8.764, shapeName=dyn_coatTail_l_03_capsule, center=<7.2381 0 7.7>, rotation=<rot 90 0 0>}]
 dyn_coatTail_l_04 [{shapeType=Capsule, length=7.0213, radius=8.764, shapeName=dyn_coatTail_l_04_capsule, center=<7.2381 0 7.7>, rotation=<rot 90 0 0>}]
 dyn_coatTail_l_05 [{shapeType=Capsule, length=7.0213, radius=8.764, shapeName=dyn_coatTail_l_05_capsule, center=<7.2381 0 7.7>, rotation=<rot 90 0 0>}]
 dyn_coatTail_l_06 [{shapeType=Capsule, length=7.0213, radius=8.764, shapeName=dyn_coatTail_l_06_capsule, center=<7.2381 0 7.7>, rotation=<rot 90 0 0>}]
 dyn_coatTail_l_MSTR [{shapeType=Capsule, length=1, radius=1, shapeName=dyn_coatTail_l_MSTR_capsule}]
 dyn_coatTail_r_01 [{shapeType=Capsule, length=7.0213, radius=8.764, shapeName=dyn_coatTail_r_01_capsule, center=<7.2381 0 7.7>, rotation=<rot 90 0 0>}]
 dyn_coatTail_r_02 [{shapeType=Capsule, length=7.0213, radius=8.764, shapeName=dyn_coatTail_r_02_capsule, center=<7.2381 0 7.7>, rotation=<rot 90 0 0>}]
 dyn_coatTail_r_03 [{shapeType=Capsule, length=7.0213, radius=8.764, shapeName=dyn_coatTail_r_03_capsule, center=<7.2381 0 7.7>, rotation=<rot 90 0 0>}]
 dyn_coatTail_r_04 [{shapeType=Capsule, length=7.0213, radius=8.764, shapeName=dyn_coatTail_r_04_capsule, center=<7.2381 0 7.7>, rotation=<rot 90 0 0>}]
 dyn_coatTail_r_05 [{shapeType=Capsule, length=7.0213, radius=8.764, shapeName=dyn_coatTail_r_05_capsule, center=<7.2381 0 7.7>, rotation=<rot 90 0 0>}]
 dyn_coatTail_r_06 [{shapeType=Capsule, length=7.0213, radius=8.764, shapeName=dyn_coatTail_r_06_capsule, center=<7.2381 0 7.7>, rotation=<rot 90 0 0>}]
 dyn_coatTail_r_MSTR [{shapeType=Capsule, length=1, radius=1, shapeName=dyn_coatTail_r_MSTR_capsule}]
 pelvis [{shapeType=Capsule, length=8.9267, radius=9.4003, shapeName=pelvis_capsule}]
 spine_01 [{shapeType=Capsule, length=6.5315, radius=9.6131, shapeName=spine_01_capsule}]
 spine_02 [{shapeType=Capsule, length=8.7584, radius=7.3862, shapeName=spine_02_capsule}]
 spine_03 [{shapeType=Capsule, length=8.7584, radius=7.3862, shapeName=spine_03_capsule}]
 thigh_l [{shapeType=Capsule, length=35.4674, radius=8.2112, shapeName=thigh_l_capsule, center=<-17.0449 0 1.2887>, rotation=<rot 90 89.6476 91.5928>}]
 thigh_r [{shapeType=Capsule, length=35.4674, radius=8.2112, shapeName=thigh_l_capsule, center=<17.0449 0 -1.2887>, rotation=<rot 90 173.169 -4.8858>}]
}
constraints [
 {bone1Name=spine_01, bone2Name=pelvis, swing1LimitDegrees=45, swing2LimitDegrees=45, twistLimitDegrees=45}
 {
  bone1Name=spine_02
  bone2Name=spine_01
  swing1LimitDegrees=45
  swing1Motion=Free
  swing2LimitDegrees=45
  swing2Motion=Free
  twistLimitDegrees=45
  twistMotion=Free
 }
 {
  bone1Name=thigh_l
  bone2Name=pelvis
  swing1LimitDegrees=45
  swing1Motion=Free
  swing2LimitDegrees=45
  swing2Motion=Free
  twistLimitDegrees=45
  twistMotion=Free
 }
 {
  bone1Name=calf_l
  bone2Name=thigh_l
  swing1LimitDegrees=45
  swing1Motion=Free
  swing2LimitDegrees=45
  swing2Motion=Free
  twistLimitDegrees=45
  twistMotion=Free
 }
 {
  bone1Name=thigh_r
  bone2Name=pelvis
  swing1LimitDegrees=45
  swing1Motion=Free
  swing2LimitDegrees=45
  swing2Motion=Free
  twistLimitDegrees=45
  twistMotion=Free
 }
 {
  bone1Name=calf_r
  bone2Name=thigh_r
  swing1LimitDegrees=45
  swing1Motion=Free
  swing2LimitDegrees=45
  swing2Motion=Free
  twistLimitDegrees=45
  twistMotion=Free
 }
 {
  bone1Name=dyn_coatTail_l_MSTR
  bone2Name=spine_02
  swing1LimitDegrees=45
  swing1Motion=Free
  swing2LimitDegrees=45
  swing2Motion=Free
  twistLimitDegrees=45
  twistMotion=Free
 }
 {bone1Name=dyn_coatTail_l_01, bone2Name=dyn_coatTail_l_MSTR, swing1LimitDegrees=10, swing2LimitDegrees=20, twistLimitDegrees=0}
 {bone1Name=dyn_coatTail_l_02, bone2Name=dyn_coatTail_l_01, swing1LimitDegrees=10, swing2LimitDegrees=20, twistLimitDegrees=0}
 {bone1Name=dyn_coatTail_l_03, bone2Name=dyn_coatTail_l_02, swing1LimitDegrees=10, swing2LimitDegrees=20, twistLimitDegrees=0}
 {bone1Name=dyn_coatTail_l_04, bone2Name=dyn_coatTail_l_03, swing1LimitDegrees=10, swing2LimitDegrees=20, twistLimitDegrees=0}
 {bone1Name=dyn_coatTail_l_05, bone2Name=dyn_coatTail_l_04, swing1LimitDegrees=10, swing2LimitDegrees=20, twistLimitDegrees=0}
 {bone1Name=dyn_coatTail_l_06, bone2Name=dyn_coatTail_l_05, swing1LimitDegrees=10, swing2LimitDegrees=20, twistLimitDegrees=0}
 {
  bone1Name=dyn_coatTail_r_MSTR
  bone2Name=spine_02
  swing1LimitDegrees=45
  swing1Motion=Free
  swing2LimitDegrees=45
  swing2Motion=Free
  twistLimitDegrees=45
  twistMotion=Free
 }
 {bone1Name=dyn_coatTail_r_01, bone2Name=dyn_coatTail_r_MSTR, swing1LimitDegrees=10, swing2LimitDegrees=20, twistLimitDegrees=0}
 {bone1Name=dyn_coatTail_r_02, bone2Name=dyn_coatTail_r_01, swing1LimitDegrees=10, swing2LimitDegrees=20, twistLimitDegrees=0}
 {bone1Name=dyn_coatTail_r_03, bone2Name=dyn_coatTail_r_02, swing1LimitDegrees=10, swing2LimitDegrees=20, twistLimitDegrees=0}
 {bone1Name=dyn_coatTail_r_04, bone2Name=dyn_coatTail_r_03, swing1LimitDegrees=10, swing2LimitDegrees=20, twistLimitDegrees=0}
 {bone1Name=dyn_coatTail_r_05, bone2Name=dyn_coatTail_r_04, swing1LimitDegrees=10, swing2LimitDegrees=20, twistLimitDegrees=0}
 {bone1Name=dyn_coatTail_r_06, bone2Name=dyn_coatTail_r_05, swing1LimitDegrees=10, swing2LimitDegrees=20, twistLimitDegrees=0}
]
