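# Physics asset for the face accessory of the male medium PrivateJet Red player character, setting up a dynamic beard.
# A six-bone center beard chain hangs from a master bone on the head, with left and right side bones branching off the second link.
# Capsules on spine, neck, head, clavicles and upper arms form the surrounding body, joined by free angular constraints.
# Beard joints use angular limits, with the widest swing near the root.
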PhysicsAsset /BRCosmetics/Characters/Player/Male/Medium/Bodies/M_MED_PrivateJet_Red/Meshes/Parts/M_MED_PrivateJet_Red_FaceAcc_Physics
bodies {
 clavicle_l [{
   shapeType=Capsule
   length=12.8647
   radius=8.3632
   shapeName=clavicle_l_capsule
   center=<4.2188 0.2731 0>
   rotation=<rot 90 -172.875 -157.6447>
  }]
 clavicle_r [{
   shapeType=Capsule
   length=12.8647
   radius=8.3632
   shapeName=clavicle_l_capsule
   center=<-4.2188 -0.2731 0>
   rotation=<rot 90 -172.875 22.3553>
  }]
 dyn_beard_c_01 [{shapeType=Capsule, length=7.3371, radius=4.7385, shapeName=dyn_beard_c_01_capsule, center=<4.3285 4.6 0>}]
 dyn_beard_c_02 [{shapeType=Capsule, length=15.8371, radius=4.7385, shapeName=dyn_beard_c_02_capsule, center=<6.6919 4.6 0>}]
 dyn_beard_c_03 [{shapeType=Capsule, length=11.1239, radius=4.7318, shapeName=dyn_beard_c_03_capsule, center=<5.036 3 0>}]
 dyn_beard_c_04 [{shapeType=Capsule, length=8.8222, radius=4.0565, shapeName=dyn_beard_c_04_capsule, center=<0 3.5 0>}]
 dyn_beard_c_05 [{shapeType=Capsule, length=4.3371, radius=4.7385, shapeName=dyn_beard_c_05_capsule, center=<0 3.5 0>}]
 dyn_beard_c_06 [{shapeType=Capsule, length=4.3371, radius=4.7385, shapeName=dyn_beard_c_06_capsule, center=<0 3.5 0>, rotation=<rot 90 0 0>}]
 dyn_beard_c_MSTR [{shapeType=Capsule, length=0.5, radius=0.5, shapeName=dyn_beard_c_MSTR_capsule}]
 dyn_beard_l_01 [{shapeType=Capsule, length=8.5, radius=3, shapeName=dyn_beard_l_01_capsule, center=<15.4032 2 0>, rotation=<rot 90 -83.9938 -62.5093>}]
 dyn_beard_r_01 [{shapeType=Capsule, length=8, radius=2.5, shapeName=dyn_beard_r_01_capsule, center=<14.0166 2 0>, rotation=<rot 90 0 0>}]
 head [{shapeType=Capsule, length=0.952, radius=10.6632, shapeName=head_capsule, center=<6.9938 0 0>}]
 neck_01 [{shapeType=Capsule, length=4.9589, radius=2.3461, shapeName=neck_01_capsule, rotation=<rot 90 0 0>}]
 neck_02 [{shapeType=Capsule, length=4.4589, radius=2.3461, shapeName=neck_02_capsule, rotation=<rot 90 0 0>}]
 spine_03 [
  {shapeType=Capsule, length=1.7937, radius=9.8215, shapeName=spine_03_capsule, center=<-0.1418 -0.5883 0>}
  {shapeType=Capsule, length=5, radius=8.5, shapeName=spine_03_capsule, center=<7.015 0 0>}
 ]
 spine_04 [{shapeType=Capsule, length=7.3673, radius=9.9488, shapeName=spine_04_capsule, center=<4.3787 -1.2599 0>}]
 spine_05 [{shapeType=Capsule, length=13.5543, radius=8.1507, shapeName=spine_05_capsule}]
 upperarm_l [{shapeType=Capsule, length=20, radius=5.505, shapeName=upperarm_l_capsule, center=<11.4035 -2.3624 -1.4234>, rotation=<rot 90 0 0>}]
 upperarm_r [{shapeType=Capsule, length=20, radius=5.505, shapeName=upperarm_l_capsule, center=<-11.4035 2.3624 1.4234>, rotation=<rot 90 0 180>}]
}
constraints [
 {
  bone1Name=head
  bone2Name=neck_02
  swing1LimitDegrees=45
  swing1Motion=Free
  swing2LimitDegrees=45
  swing2Motion=Free
  twistLimitDegrees=45
  twistMotion=Free
 }
 {
  bone1Name=spine_04
  bone2Name=spine_03
  swing1LimitDegrees=45
  swing1Motion=Free
  swing2LimitDegrees=45
  swing2Motion=Free
  twistLimitDegrees=45
  twistMotion=Free
 }
 {
  bone1Name=spine_05
  bone2Name=spine_04
  swing1LimitDegrees=45
  swing1Motion=Free
  swing2LimitDegrees=45
  swing2Motion=Free
  twistLimitDegrees=45
  twistMotion=Free
 }
 {
  bone1Name=clavicle_l
  bone2Name=spine_05
  swing1LimitDegrees=45
  swing1Motion=Free
  swing2LimitDegrees=45
  swing2Motion=Free
  twistLimitDegrees=45
  twistMotion=Free
 }
 {
  bone1Name=neck_01
  bone2Name=spine_05
  swing1LimitDegrees=45
  swing1Motion=Free
  swing2LimitDegrees=45
  swing2Motion=Free
  twistLimitDegrees=45
  twistMotion=Free
 }
 {
  bone1Name=neck_02
  bone2Name=neck_01
  swing1LimitDegrees=45
  swing1Motion=Free
  swing2LimitDegrees=45
  swing2Motion=Free
  twistLimitDegrees=45
  twistMotion=Free
 }
 {
  bone1Name=dyn_beard_c_MSTR
  bone2Name=head
  swing1LimitDegrees=45
  swing1Motion=Free
  swing2LimitDegrees=45
  swing2Motion=Free
  twistLimitDegrees=45
  twistMotion=Free
 }
 {bone1Name=dyn_beard_c_01, bone2Name=dyn_beard_c_MSTR, swing1LimitDegrees=70, swing2LimitDegrees=8, twistLimitDegrees=10}
 {bone1Name=dyn_beard_c_02, bone2Name=dyn_beard_c_01, swing1LimitDegrees=70, swing2LimitDegrees=8, twistLimitDegrees=10}
 {bone1Name=dyn_beard_c_03, bone2Name=dyn_beard_c_02, swing1LimitDegrees=55, swing2LimitDegrees=20, twistLimitDegrees=10}
 {bone1Name=dyn_beard_c_04, bone2Name=dyn_beard_c_03, swing1LimitDegrees=50, swing2LimitDegrees=20, twistLimitDegrees=10}
 {bone1Name=dyn_beard_c_05, bone2Name=dyn_beard_c_04, swing1LimitDegrees=40, swing2LimitDegrees=20, twistLimitDegrees=10}
 {bone1Name=dyn_beard_c_06, bone2Name=dyn_beard_c_05, swing1LimitDegrees=40, swing2LimitDegrees=20, twistLimitDegrees=10}
 {
  bone1Name=upperarm_l
  bone2Name=clavicle_l
  swing1LimitDegrees=45
  swing1Motion=Free
  swing2LimitDegrees=45
  swing2Motion=Free
  twistLimitDegrees=45
  twistMotion=Free
 }
 {
  bone1Name=clavicle_r
  bone2Name=spine_05
  swing1LimitDegrees=45
  swing1Motion=Free
  swing2LimitDegrees=45
  swing2Motion=Free
  twistLimitDegrees=45
  twistMotion=Free
 }
 {bone1Name=dyn_beard_l_01, bone2Name=dyn_beard_c_02, swing1LimitDegrees=20, swing2LimitDegrees=20, twistLimitDegrees=0}
 {bone1Name=dyn_beard_r_01, bone2Name=dyn_beard_c_02, swing1LimitDegrees=20, swing2LimitDegrees=20, twistLimitDegrees=0}
 {
  bone1Name=upperarm_r
  bone2Name=clavicle_r
  swing1LimitDegrees=45
  swing1Motion=Free
  swing2LimitDegrees=45
  swing2Motion=Free
  twistLimitDegrees=45
  twistMotion=Free
 }
]
