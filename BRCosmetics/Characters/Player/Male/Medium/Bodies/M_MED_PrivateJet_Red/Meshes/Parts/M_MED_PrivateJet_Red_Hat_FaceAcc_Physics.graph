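# Physics asset for the hat of the red PrivateJet male medium player cosmetic, simulating secondary motion on its dynamic bones.
# A six-bone center chain, generally looser toward the tip, and four single-bone fur pieces (front, back, left, right) hang from master bodies linked to the head.
# Fur swings narrowly on one axis.
# Head, neck, spine, clavicle and upper-arm capsules form a freely constrained upper-body shape, likely for hat collision.
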PhysicsAsset /BRCosmetics/Characters/Player/Male/Medium/Bodies/M_MED_PrivateJet_Red/Meshes/Parts/M_MED_PrivateJet_Red_Hat_FaceAcc_Physics
bodies {
 clavicle_l [{
   shapeType=Capsule
   length=13.8647
   radius=7.3632
   shapeName=clavicle_l_capsule
   center=<4.4335 -0.7451 -1.3509>
   rotation=<rot 90 0 -11.9076>
  }]
 clavicle_r [{
   shapeType=Capsule
   length=13.8647
   radius=7.3632
   shapeName=clavicle_l_capsule
   center=<-4.4335 0.7451 1.4364>
   rotation=<rot 90 -2.8624 165.23>
  }]
 dyn_hatFur_back_01 [{shapeType=Capsule, length=2.422, radius=5.8903, shapeName=dyn_hatFur_back_01_capsule, center=<11.0168 6 0>, rotation=<rot 90 0 0>}]
 dyn_hatFur_back_MSTR [{shapeType=Capsule, length=0.5, radius=0.5, shapeName=dyn_hatFur_back_MSTR_capsule}]
 dyn_hatFur_front_01 [{shapeType=Capsule, length=2.422, radius=5.8903, shapeName=dyn_hatFur_front_01_capsule, center=<11.0168 6 0>, rotation=<rot 90 0 0>}]
 dyn_hatFur_front_MSTR [{shapeType=Capsule, length=0.5, radius=0.5, shapeName=dyn_hatFur_front_MSTR_capsule}]
 dyn_hatFur_l_01 [{shapeType=Capsule, length=2.422, radius=5.8903, shapeName=dyn_hatFur_l_01_capsule, center=<11.0168 6 0>, rotation=<rot 90 0 0>}]
 dyn_hatFur_l_MSTR [{shapeType=Capsule, length=0.5, radius=0.5, shapeName=dyn_hatFur_l_MSTR_capsule}]
 dyn_hatFur_r_01 [{shapeType=Capsule, length=2.422, radius=5.8903, shapeName=dyn_hatFur_r_01_capsule, center=<11.0168 6 0>, rotation=<rot 90 0 0>}]
 dyn_hatFur_r_MSTR [{shapeType=Capsule, length=0.5, radius=0.5, shapeName=dyn_hatFur_r_MSTR_capsule}]
 dyn_hat_c_01 [{shapeType=Capsule, length=0, radius=4.5846, shapeName=dyn_hat_c_01_capsule, center=<1.539 0 0>, rotation=<rot 90 0 0>}]
 dyn_hat_c_02 [{shapeType=Capsule, length=1, radius=5.5568, shapeName=dyn_hat_c_02_capsule, center=<1.2374 0 0>, rotation=<rot 90 0 0>}]
 dyn_hat_c_03 [{shapeType=Capsule, length=11.529, radius=4.0846, shapeName=dyn_hat_c_03_capsule, center=<8 0 0>, rotation=<rot 90 0 0>}]
 dyn_hat_c_04 [{shapeType=Capsule, length=7.6913, radius=3.0433, shapeName=dyn_hat_c_04_capsule, center=<5 0 0>, rotation=<rot 90 0 0>}]
 dyn_hat_c_05 [{shapeType=Capsule, length=1.9234, radius=2.0089, shapeName=dyn_hat_c_05_capsule, center=<1.2326 0 0>, rotation=<rot 90 180 180>}]
 dyn_hat_c_06 [{shapeType=Capsule, length=2, radius=4.9416, shapeName=dyn_hat_c_06_capsule, center=<5.2691 0 1.5>, rotation=<rot 90 0 0>}]
 dyn_hat_c_MSTR [{shapeType=Capsule, length=0.5, radius=0.5, shapeName=dyn_hat_c_MSTR_capsule}]
 head [{shapeType=Capsule, length=0.952, radius=10.6632, shapeName=head_capsule, center=<6.9938 0 0>}]
 neck_01 [{shapeType=Capsule, length=4.9589, radius=2.3461, shapeName=neck_01_capsule, rotation=<rot 90 0 0>}]
 neck_02 [{shapeType=Capsule, length=4.4589, radius=2.3461, shapeName=neck_02_capsule, rotation=<rot 90 0 0>}]
 spine_03 [{shapeType=Capsule, length=4.5, radius=9, shapeName=spine_03_capsule, center=<7.015 1.4161 0>}]
 spine_05 [{shapeType=Capsule, length=12.5543, radius=9.1507, shapeName=spine_05_capsule}]
 upperarm_l [{shapeType=Capsule, length=21.5, radius=4.005, shapeName=upperarm_l_capsule, center=<11.4035 0.0371 -1.4234>, rotation=<rot 90 0 0>}]
 upperarm_r [{shapeType=Capsule, length=21.5, radius=4.005, shapeName=upperarm_l_capsule, center=<-11.4035 -0.0371 1.4234>, rotation=<rot 90 0 180>}]
}
constraints [
 {
  bone1Name=head
  bone2Name=neck_02
  swing1LimitDegrees=45
  swing1Motion=Free
  swing2LimitDegrees=45
  swing2Motion=Free
  twistLimitDegrees=45
  twistMotion=Free
 }
 {
  bone1Name=spine_05
  bone2Name=spine_03
  swing1LimitDegrees=45
  swing1Motion=Free
  swing2LimitDegrees=45
  swing2Motion=Free
  twistLimitDegrees=45
  twistMotion=Free
 }
 {
  bone1Name=clavicle_l
  bone2Name=spine_05
  swing1LimitDegrees=45
  swing1Motion=Free
  swing2LimitDegrees=45
  swing2Motion=Free
  twistLimitDegrees=45
  twistMotion=Free
 }
 {
  bone1Name=neck_01
  bone2Name=spine_05
  swing1LimitDegrees=45
  swing1Motion=Free
  swing2LimitDegrees=45
  swing2Motion=Free
  twistLimitDegrees=45
  twistMotion=Free
 }
 {
  bone1Name=neck_02
  bone2Name=neck_01
  swing1LimitDegrees=45
  swing1Motion=Free
  swing2LimitDegrees=45
  swing2Motion=Free
  twistLimitDegrees=45
  twistMotion=Free
 }
 {
  bone1Name=dyn_hatFur_back_MSTR
  bone2Name=head
  swing1LimitDegrees=45
  swing1Motion=Free
  swing2LimitDegrees=45
  swing2Motion=Free
  twistLimitDegrees=45
  twistMotion=Free
 }
 {bone1Name=dyn_hatFur_back_01, bone2Name=dyn_hatFur_back_MSTR, swing1LimitDegrees=15, swing2LimitDegrees=0, twistLimitDegrees=0}
 {
  bone1Name=dyn_hatFur_front_MSTR
  bone2Name=head
  swing1LimitDegrees=45
  swing1Motion=Free
  swing2LimitDegrees=45
  swing2Motion=Free
  twistLimitDegrees=45
  twistMotion=Free
 }
 {bone1Name=dyn_hatFur_front_01, bone2Name=dyn_hatFur_front_MSTR, swing1LimitDegrees=15, swing2LimitDegrees=0, twistLimitDegrees=0}
 {
  bone1Name=dyn_hatFur_l_MSTR
  bone2Name=head
  swing1LimitDegrees=45
  swing1Motion=Free
  swing2LimitDegrees=45
  swing2Motion=Free
  twistLimitDegrees=45
  twistMotion=Free
 }
 {bone1Name=dyn_hatFur_l_01, bone2Name=dyn_hatFur_l_MSTR, swing1LimitDegrees=15, swing2LimitDegrees=0, twistLimitDegrees=0}
 {
  bone1Name=dyn_hatFur_r_MSTR
  bone2Name=head
  swing1LimitDegrees=45
  swing1Motion=Free
  swing2LimitDegrees=45
  swing2Motion=Free
  twistLimitDegrees=45
  twistMotion=Free
 }
 {bone1Name=dyn_hatFur_r_01, bone2Name=dyn_hatFur_r_MSTR, swing1LimitDegrees=15, swing2LimitDegrees=0, twistLimitDegrees=0}
 {
  bone1Name=dyn_hat_c_MSTR
  bone2Name=head
  swing1LimitDegrees=45
  swing1Motion=Free
  swing2LimitDegrees=45
  swing2Motion=Free
  twistLimitDegrees=45
  twistMotion=Free
 }
 {bone1Name=dyn_hat_c_01, bone2Name=dyn_hat_c_MSTR, swing1LimitDegrees=20, swing2LimitDegrees=10, twistLimitDegrees=0}
 {bone1Name=dyn_hat_c_02, bone2Name=dyn_hat_c_01, swing1LimitDegrees=10, swing2LimitDegrees=10, twistLimitDegrees=0}
 {bone1Name=dyn_hat_c_03, bone2Name=dyn_hat_c_02, swing1LimitDegrees=20, swing2LimitDegrees=8, twistLimitDegrees=10}
 {bone1Name=dyn_hat_c_04, bone2Name=dyn_hat_c_03, swing1LimitDegrees=30, swing2LimitDegrees=20, twistLimitDegrees=10}
 {bone1Name=dyn_hat_c_05, bone2Name=dyn_hat_c_04, swing1LimitDegrees=30, swing2LimitDegrees=20, twistLimitDegrees=10}
 {bone1Name=dyn_hat_c_06, bone2Name=dyn_hat_c_05, swing1LimitDegrees=30, swing2LimitDegrees=10, twistLimitDegrees=0}
 {
  bone1Name=upperarm_l
  bone2Name=clavicle_l
  swing1LimitDegrees=45
  swing1Motion=Free
  swing2LimitDegrees=45
  swing2Motion=Free
  twistLimitDegrees=45
  twistMotion=Free
 }
 {
  bone1Name=upperarm_r
  bone2Name=clavicle_r
  swing1LimitDegrees=45
  swing1Motion=Free
  swing2LimitDegrees=45
  swing2Motion=Free
  twistLimitDegrees=45
  twistMotion=Free
 }
 {
  bone1Name=clavicle_r
  bone2Name=spine_05
  swing1LimitDegrees=45
  swing1Motion=Free
  swing2LimitDegrees=45
  swing2Motion=Free
  twistLimitDegrees=45
  twistMotion=Free
 }
]
