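# Face accessory AnimBP for the PrivateJet Red male character, adding secondary motion to beard, hat and hat fur.
# Copied parent and body poses feed a beard RigidBody, then a hat RigidBody selected by backbling state.
# The emote_BuiltIn curve blends in slot animation and fades out physics during emotes.
# Low LOD or disabled rigid bodies switch on beard bone constraints, while material WPO functions receive hat and head motion.
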AnimBlueprint /BRCosmetics/Characters/Player/Male/Medium/Bodies/M_MED_PrivateJet_Red/Meshes/Parts/M_MED_PrivateJet_Red_FaceAcc_AnimBP
target_skeleton=/Game/Characters/Player/Male/Medium/Bodies/M_MED_PrivateJet_Red/Meshes/Parts/M_MED_PrivateJet_Red_FaceAcc_Skeleton
graphs [
 AnimGraph {nodes [
   n1 = Get curveInverseValue() -> (curveInverseValue:double)
   n2 = "Constraint - dyn_beard_c_01 (PP)"!(ComponentPose=Local To Component(LocalPose=Use cached pose 'fullCache'()), Alpha=@n1, ConstraintWeights_0=1, ConstraintWeights_1:float=1) -> (Pose) {
    alphaBoolBlend {blendOption=Linear}
    boneToModify {boneName=dyn_beard_c_01}
    constraintSetup [{targetBone {boneName=spine_05}}, {targetBone {boneName=head}}]
    constraintWeights [1,1]
   }
   n3 = "Constraint - dyn_beard_c_02 (PP)"!(ComponentPose=@n2, Alpha=@n1, ConstraintWeights_0=0.6, ConstraintWeights_1:float=0.4) -> (Pose) {
    alphaBoolBlend {blendOption=Linear}
    boneToModify {boneName=dyn_beard_c_02}
    constraintSetup [{targetBone {boneName=spine_05}}, {targetBone {boneName=head}}]
    constraintWeights [1,1]
   }
   n4 = "Constraint - dyn_beard_c_03 (P)"(ComponentPose=@n3, Alpha=@n1, ConstraintWeights_0=1) {
    alphaBoolBlend {blendOption=Linear}
    boneToModify {boneName=dyn_beard_c_03}
    constraintSetup [{targetBone {boneName=spine_05}}]
    constraintWeights [1]
   }
   Output Pose(Result=Blend Poses by bool(
     BlendPose_0=Component To Local(ComponentPose=@n4)
     BlendPose_1=Use cached pose 'fullCache'()
     bActiveValue=Get isLowLOD orRBANDisabled() -> (isLowLOD orRBANDisabled:bool)
    ) {blendType=Linear}) {layerGroup=None}
   n5 = Layered blend per bone(
    BasePose=Copy Pose From Mesh() {bCopyCurves=true, bUseAttachedParent=true}
    BlendPoses_0=Copy Pose From Mesh(SourceMeshComponent=Get Body_Mesh())
   ) {layerSetup [{branchFilters [{boneName=upperarm_l}, {boneName=upperarm_r}, {boneName=clavicle_l}, {boneName=clavicle_r}]}]}
   Save cached pose 'basePoseCache'(Pose=@n5) {cacheName=basePoseCache}
   n6 = Layered blend per bone(
    BasePose=Use cached pose 'basePoseCache'()
    BlendPoses_0=Slot 'DefaultSlot'()
    BlendWeights_0=Get buildInEmoteDetector() -> (buildInEmoteDetector:double)
   ) {layerSetup [{branchFilters [
       {boneName=dyn_beard_c_MSTR}
       {boneName=dyn_hat_c_MSTR}
       {boneName=dyn_hatFur_back_MSTR}
       {boneName=dyn_hatFur_front_MSTR}
       {boneName=dyn_hatFur_l_MSTR}
       {boneName=dyn_hatFur_r_MSTR}
      ]}]}
   n7 = RigidBody!(ComponentPose=Local To Component(LocalPose=@n6), Alpha=Get curveInverseValue(), OverrideWorldGravity:Vector="0,0,-980", ExternalForce) -> (Pose) {
    bClampLinearTranslationLimitToRefPose=true
    bOverrideWorldGravity=true
    overridePhysicsAsset=&/M_MED_PrivateJet_Red_FaceAcc_Physics
    simulationSpace=BaseBoneSpace
    alphaBoolBlend {blendOption=Linear}
    alphaScaleBiasClamp {bInterpResult=true, interpSpeedDecreasing=2, interpSpeedIncreasing=0}
    baseBoneRef {boneName=neck_01}
    componentAppliedLinearAccClamp=<6000 6000 6000>
    componentLinearAccScale=<0.4 0.6 0.6>
    componentLinearVelScale=<0.4 0.6 0.6>
   }
   Save cached pose 'beardCache'(Pose=Component To Local(ComponentPose=@n7)) {cacheName=beardCache}
   n8 = RigidBody(
    ComponentPose=Local To Component(LocalPose=Use cached pose 'beardCache'())
    Alpha=Get curveInverseValue()
    OverrideWorldGravity=Get Gravity Override()
   ) {
    bClampLinearTranslationLimitToRefPose=true
    bOverrideWorldGravity=true
    overridePhysicsAsset=&/M_MED_PrivateJet_Red_Hat_backbling_FaceAcc_Physics
    simulationSpace=BaseBoneSpace
    alphaBoolBlend {blendOption=Linear}
    alphaScaleBiasClamp {bInterpResult=true, interpSpeedDecreasing=2, interpSpeedIncreasing=0}
    baseBoneRef {boneName=head}
    componentAppliedLinearAccClamp=<6000 6000 6000>
    componentLinearAccScale=<0.6 0.55 0.55>
    componentLinearVelScale=<0.6 0.55 0.55>
   }
   n9 = RigidBody(
    ComponentPose=Local To Component(LocalPose=Use cached pose 'beardCache'())
    Alpha=Get curveInverseValue()
    OverrideWorldGravity=Get Gravity Override()
   ) {
    bClampLinearTranslationLimitToRefPose=true
    bOverrideWorldGravity=true
    overridePhysicsAsset=&/M_MED_PrivateJet_Red_Hat_FaceAcc_Physics
    simulationSpace=BaseBoneSpace
    alphaBoolBlend {blendOption=Linear}
    alphaScaleBiasClamp {bInterpResult=true, interpSpeedDecreasing=2, interpSpeedIncreasing=0}
    baseBoneRef {boneName=head}
    componentAppliedLinearAccClamp=<6000 6000 6000>
    componentLinearAccScale=<0.6 0.55 0.55>
    componentLinearVelScale=<0.6 0.55 0.55>
   }
   Save cached pose 'fullCache'(Pose=Blend Poses by bool(
     BlendPose_0=Component To Local(ComponentPose=@n8)
     BlendPose_1=Component To Local(ComponentPose=@n9)
     bActiveValue=Get Is Wearing Backbling()
    ) {blendType=Linear}) {cacheName=fullCache}
   Get Gravity Override()
  ]}
 EventGraph {nodes [
   n1 = Event BlueprintUpdateAnimation()
   > Set Delta Time X(Delta Time X=@n1.DeltaTimeX)
   > n2 = Sequence!(execute) -> (then_0, then_1, then_2:Exec, then_3:Exec, then_4:Exec, then_5:Exec, then_6:Exec)
   > n3 = Gravity Override State Function DEPRECATED(CharacterPartAnimInstance=Self-Reference(), GameTimeInSeconds=Get Delta Time X(), GravityOverrideParams=Get Gravity Override Params())
   > Set Gravity Override(Gravity Override=@n3)
   Set Is Wearing Backbling(execute=@n2.then_1, Is Wearing Backbling=CheckForBackbling())
   n4 = Branch(execute=@n2.then_2, Condition=Get bIsPlayingEmote())
   > n5 = Branch(Condition=Get bIsInFrontEnd())
   n6 = GetCurveValue(self=GetOwnerFrontEndAnimInstance(), CurveName=emote_BuiltIn)
   > Set buildInEmoteDetector(buildInEmoteDetector:double=@n6) -> (Output_Get:double)
   > Set curveInverseValue(curveInverseValue:double=Remap Float(Input=@n6, OldRange=(X=0,Y=1), NewRange=(X=1,Y=0)) -> (Output=~)) -> (Output_Get:double)
   n7 = GetCurveValue(
    self=GetOwnerFortPlayerAnimInstance!(self:Custom Character Part Anim Instance Object Reference) -> (ReturnValue:Fort Player Anim Instance Object Reference)
    CurveName=emote_BuiltIn
   )
   Set buildInEmoteDetector(execute=@n5.else, buildInEmoteDetector=@n7)
   > Set curveInverseValue(curveInverseValue=Remap Float(Input=@n7, OldRange=(X=0,Y=1), NewRange=(X=1,Y=0)) -> (Output=~))
   Set curveInverseValue(execute=@n4.else, curveInverseValue=1)
   n8 = Get CurrentLOD()
   n9 = "Equal (Integer)"(A=@n8, B=3)
   n10 = "Equal (Integer)"(A=@n8, B=2)
   n11 = NOT Boolean(A=Get bRigidBodyEnabled())
   n12 = "Equal (Integer)"(A=@n8, B=1)
   Set isLowLOD orRBANDisabled(execute=@n2.then_4, isLowLOD orRBANDisabled:bool=OR Boolean(A=OR Boolean(A=@n10, B=@n9), B=AND Boolean(A=@n11, B=@n12))) -> (Output_Get:bool)
   Set WPO Globals Function(
    execute=@n2.then_3
    Material=Get matInstanceDynamic() -> (matInstanceDynamic:Material Instance Dynamic Object Reference)
    Skeletal Mesh=Get Body_Mesh()
    "Local Noise Transform  Location"="Get Local Noise Transform  Location"() -> ("Local Noise Transform  Location":Vector)
    Local Noise Offset=Get Local Noise Offset()
   )
   > Set WPO Regional Function!(
    execute
    Material:Material Instance Dynamic Object Reference=Get matInstanceDynamic()
    Overlay Material:Material Instance Dynamic Object Reference
    Skeletal Mesh:Skeletal Mesh Component Object Reference=Get Body_Mesh()
    Velocity Transform:Name=dyn_hat_c_03
    Velocity Interp Speed:double=Get Velocity Interp Speed()
    Collision Transform:Name=None
    Collision Transform Axis:EAxis Enum
    Invert Collision Transform Axis:bool=false
    Delta Time X:double=Get Delta Time X()
    Previous Velocity Transform Location:"Vector (by ref)"=Get Previous Velocity Transform Location() -> (Previous Velocity Transform Location:Vector)
    Previous Velocity Transform Velocity:"Vector (by ref)"=Get Previous Velocity Transform Velocity() -> (Previous Velocity Transform Velocity:Vector)
    Region:int=1
   ) -> (then)
   > Set WPO Regional Function(
    Material=Get matInstanceDynamic()
    Skeletal Mesh=Get Body_Mesh()
    Velocity Transform=head
    Velocity Interp Speed=Get Velocity Interp Speed()
    Collision Transform=None
    Invert Collision Transform Axis=false
    Delta Time X=Get Delta Time X()
    Previous Velocity Transform Location=Get Previous Velocity Transform Location_0() -> (Previous Velocity Transform Location_0:Vector)
    Previous Velocity Transform Velocity=Get Previous Velocity Transform Velocity_0() -> (Previous Velocity Transform Velocity_0:Vector)
    Region=2
   )
   Event OnCharacterCustomizationLoaded()
   > n13 = Sequence!(execute) -> (then_0, then_1)
   > n14 = Cast To PlayerPawn_Athena(Object=TryGetPawnOwner())
   > Set Body_Mesh(Body_Mesh=GetSkeletalMeshForPartType(self=@n14))
   n15 = Set Body_Mesh(execute=@n13.then_1, Body_Mesh=GetPartSkeletalMeshComponent())
   > n16 = Get Dynamic Material Instance(SkeletalMeshComponent=@n15)
   > Set matInstanceDynamic(matInstanceDynamic:Material Instance Dynamic Object Reference=@n16) -> (Output_Get:Material Instance Dynamic Object Reference)
   n17 = Branch(Condition=Get bIsPlayingEmote())
   Do Once(execute=@n17.else, Reset=@n17.then, Start Closed=true)
   > Delay(Duration=0)
   > ResetDynamics(InTeleportType=TeleportPhysics)
   > ResetDynamics(InTeleportType=ResetPhysics)
  ]}
]
variables [
 Alpha Output {type=number}
 Body_Mesh {type=SkeletalMeshComponent}
 buildInEmoteDetector {type=number}
 curveInverseValue {type=number}
 Delta Time X {type=number}
 enableConstraintBeard {type=boolean}
 Gravity Override {type=Vector}
 Gravity Override Params {type=GravityOverrideParamsStruct}
 Is Wearing Backbling {type=boolean}
 isLowLOD orRBANDisabled {type=boolean}
 Local Noise Offset {type=Vector}
 "Local Noise Transform  Location" {type=Vector}
 matInstanceDynamic {type=MaterialInstanceDynamic}
 Previous Velocity Transform Location {type=Vector}
 Previous Velocity Transform Location_0 {type=Vector}
 Previous Velocity Transform Velocity {type=Vector}
 Previous Velocity Transform Velocity_0 {type=Vector}
 Velocity Interp Speed {type=number}
]
